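# Additive unlit Niagara mesh material for a zero-point scan ping: an expanding colored pulse ring projected onto scene geometry behind the mesh.
# Distance from the object to the world position behind translucency builds the ring mask and hot edge, faded near the camera.
# Scene-depth edge detection, limited to Epic quality on deferred shading, brightens outlines inside the ring.
# A Fresnel glow driven by dynamic parameters is added, then tinted by particle color with optional depth fade.
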Material /BattlePassS21/QuestActors/Narrative/Week_02/Vfx/PulseEffect/MAT_ZeroPointScan_MeshPing
graph {
 properties {
  bCastRayTracedShadows=false
  bUsedWithInstancedStaticMeshes=true
  bUsedWithNiagaraMeshParticles=true
  bUsedWithSplineMeshes=true
  blendMode=BLEND_Additive
  pixelDepthOffsetMode=PDOM_Legacy
  shadingModel=MSM_Unlit
  usedShadingModels=Unlit
 }
 nodes [
  Constant() {r=1}
  Constant() {r=1}
  DynamicParameter() {paramNames [WhateverTime, TimeInterval, FresnelBrightness, FresnelExponent]}
  n1 = ParticleColor()
  n2 = ObjectPositionWS()
  n3 = MaterialFunctionCall() {materialFunction=&WorldPositionBehindTranslucency}
  n4 = MaterialFunctionCall(Vector 3=Subtract(A=@n3.Result, B=@n2)) {materialFunction=/Engine/Functions/Engine_MaterialFunctions02/Utility/VectorLength}
  n5 = NamedRerouteUsage()
  n6 = ScalarParameter() {defaultValue=200, desc=Controls the thickness of the ping ring in World Units, group=Mask Shape, parameterName=PingRingThickness}
  n7 = Subtract(A=@n5, B=@n6) {constB=50}
  n8 = MaterialFunctionCall(
   Input=@n4.V3 Length
   Input Low=@n7
   Input High=Subtract(A=@n5, B=ScalarParameter() {group=Mask Shape, parameterName=PingRingOuterEdgeNudge})
   Target Low=Constant()
   Target High=Constant() {r=1}
  ) {materialFunction=&RemapValueRange}
  n9 = If(A=@n8.Result, "A > B"=Constant(), "A == B"=@n8.Result, "A < B"=@n8.Result) {constB=1, desc=">1 means past what we want as an edge"}
  n10 = SphereMask(A=@n3.Result, B=@n2, Radius=@n5) {attenuationRadius=200}
  n11 = ComponentMask(None=CameraPositionWS()) {g=true, r=true}
  n12 = ComponentMask(None=ObjectPositionWS()) {g=true, r=true}
  n13 = Multiply(A=Normalize(VectorInput=Subtract(A=@n11, B=@n12)), B=NamedRerouteUsage())
  n14 = MaterialFunctionCall(Vector 2=Subtract(A=@n11, B=Add(A=@n12, B=@n13))) {materialFunction=&VectorLength}
  n15 = ScalarParameter() {
   defaultValue=500
   desc=Distance in World Units to begin fading the effect as it approachs the camera
   group=ScooterPingControls
   parameterName=DistanceToBeginCameraFade
  }
  n16 = Multiply(A=Multiply(A=@n10, B=Saturate(None=@n9)) {constA=1}, B=Saturate(None=Divide(A=@n14.V2 Length, B=@n15) {constB=500}))
  n17 = ScalarParameter() {
   defaultValue=1
   desc=Controls in conjuction with PulseEdgeHotness the thickness of the brightened edge
   group=ScooterPingControls
   parameterName=PulseEdgeHotnessThickness
  }
  n18 = ScalarParameter() {defaultValue=2, desc=Ramps the Brightness of the far edge of the effect, group=ScooterPingControls, parameterName=PulseEdgeHotness}
  n19 = Power(Base=Multiply(A=NamedRerouteUsage(), B=@n17) {constB=2}, Exp=@n18) {constExponent=10}
  n20 = VectorParameter() {desc="Controls Color For ", group=ScooterPingControls, parameterName=PulseColor, defaultValue=<rgba 0.0382 0.4072 0.8469 1>}
  n21 = Constant() {r=1}
  n22 = ScreenPosition()
  n23 = SceneTexelSize()
  n24 = Multiply(A=@n23, B=Constant2Vector() {g=-1, r=-1})
  n25 = SceneDepth()
  n26 = Subtract(A=SceneDepth(UVs=Add(A=@n22.ViewportUV, B=@n24)).R, B=@n25.R)
  n27 = Multiply(A=@n23, B=Constant2Vector() {g=-1, r=1})
  n28 = Subtract(A=SceneDepth(UVs=Add(A=@n22.ViewportUV, B=@n27)).R, B=@n25.R)
  n29 = Constant2Vector() {g=1, r=1}
  n30 = SceneDepth(UVs=Add(A=@n22.ViewportUV, B=Multiply(A=@n23, B=@n29)))
  n31 = Add(A=@n22.ViewportUV, B=Multiply(A=@n23, B=Constant2Vector() {g=1, r=-1}))
  n32 = Add(A=Add(A=Add(A=@n26, B=@n28), B=Subtract(A=@n30.R, B=@n25.R)), B=Subtract(A=SceneDepth(UVs=@n31).R, B=@n25.R))
  n33 = Saturate(None=ComponentMask(None=Abs(None=@n32)) {r=true})
  n34 = ScalarParameter() {
   defaultValue=1305.9292
   desc=Power Control for Edge Detection to limit depth effect
   group=ScooterPingControls
   parameterName=EdgeFalloffRate
  }
  n35 = QualitySwitch(Default=@n21, Epic=ShadingPathSwitch(Default=@n21, Deferred=Power(Base=@n33, Exp=@n34) {constExponent=20}))
  n36 = Add(
   A=Multiply(
    A=Multiply(A=@n19, B=@n20.RGB)
    B=ScalarParameter() {defaultValue=1, desc=Overall Ping Effect Brightness, group=ScooterPingControls, parameterName=EffectBrightness}
   )
   B=Multiply(
    A=Multiply(A=@n20.RGB, B=@n35)
    B=ScalarParameter() {defaultValue=3, desc=Brightness of the Edge Detection, group=ScooterPingControls, parameterName=EdgeBrightness}
   ) {constB=3}
  )
  n37 = DynamicParameter() {paramNames [WhateverTime, TimeInterval, FresnelBrightness, FresnelExponent]}
  n38 = Fresnel(
   ExponentIn=Multiply(A=@n37.FresnelExponent, B=ScalarParameter() {defaultValue=5, group=Fresnel, parameterName=FresnelExp})
   BaseReflectFractionIn=ScalarParameter() {defaultValue=0.01, group=Fresnel, parameterName=FresnelVaseReflect, sliderMax=1}
  ) {baseReflectFraction=0.01, exponent=1.5595}
  n39 = ScalarParameter() {defaultValue=5, group=Fresnel, parameterName=FresnelBrightness}
  n40 = Multiply(A=Multiply(A=Saturate(None=@n38), B=Multiply(A=@n37.FresnelBrightness, B=@n39)), B=NamedRerouteUsage())
  n41 = Multiply(A=@n1.RGB, B=Add(A=Multiply(A=@n36, B=@n16), B=Multiply(A=@n20.RGB, B=@n40)) {desc=Add Fresnel Glow after edge masking})
  n42 = DepthFade(FadeDistance=ScalarParameter() {parameterName=FadeDistance})
  n43 = Multiply(A=ScalarParameter() {defaultValue=1, parameterName=DEpthEdgeEmssiveMult}, B=OneMinus(None=@n42))
  MaterialFunctionCall(Emissive=StaticSwitchParameter(True=Multiply(A=@n41, B=@n43), False=@n41) {parameterName=UseDepthFade}) {materialFunction=&MF_FX_EmissiveAdjustments}
  n44 = CameraVectorWS()
  n45 = VertexNormalWS()
  n46 = Saturate(None=Abs(None=DotProduct(A=@n44, B=@n45)))
  n47 = DynamicParameter() {paramNames [WhateverTime, TimeInterval, FresnelBrightness, FresnelExponent]}
  n48 = ScalarParameter() {defaultValue=5, group=Fresnel, parameterName=FresnelExp}
  Multiply(
   A=Power(Base=OneMinus(None=@n46), Exp=Multiply(A=@n47.FresnelExponent, B=@n48))
   B=Multiply(A=@n47.FresnelBrightness, B=ScalarParameter() {defaultValue=5, group=Fresnel, parameterName=FresnelBrightness})
  )
  n49 = WorldPosition()
  n50 = ObjectPositionWS()
  NamedRerouteDeclaration(None=MaterialFunctionCall(Vector 3=Subtract(A=@n49.XYZ, B=@n50)) {materialFunction=&VectorLength}.V3 Length) {name=Radius, nodeColor=<rgba 0.8281 1 0 1>}
  n51 = DynamicParameter() {paramNames [WhateverTime, TimeInterval, FresnelBrightness, FresnelExponent]}
  NamedRerouteDeclaration(None=Saturate(None=OneMinus(None=@n51.WhateverTime) {desc=Plug Debugs into Me})) {name=PulseLerpRef, nodeColor=<rgba 0.8516 0 1 1>}
  n52 = ComponentMask(None=WorldPosition().XYZ) {b=true}
  n53 = ComponentMask(None=ObjectPositionWS()) {b=true}
  n54 = ScalarParameter() {defaultValue=800, desc=World Height from object to begin fading effect in the Z., group=ScooterPingControls, parameterName=FadeHeight}
  n55 = Divide(A=Abs(None=Subtract(A=@n52, B=@n53)), B=@n54) {constB=800}
  NamedRerouteDeclaration(None=OneMinus(None=Saturate(None=@n55))) {name=ZMasking, nodeColor=<rgba 1 0 0.0703 1>}
  n56 = ScalarParameter() {desc=Offsets Radius of trailing edge for Emissive Heat spike, group=Edge Heat Controls, parameterName=AdjustHotEdgeInner}
  n57 = ScalarParameter() {desc=Offsets Radius of leading edge for Emissive Heat spike, group=Edge Heat Controls, parameterName=Adjust HotEdgerOuter}
  n58 = Saturate(None=MaterialFunctionCall(Input=@n4.V3 Length, Input Low=Add(A=@n7, B=@n56), Input High=Add(A=@n57, B=@n5), Target Low=Constant(), Target High=Constant() {r=1}) {materialFunction=&RemapValueRange}.Result)
  NamedRerouteDeclaration(None=Multiply(A=@n10, B=Sine(None=@n58) {period=2})) {name=EdgeHottnessMask, nodeColor=<rgba 1 0.2109 0 1>}
  n59 = Power(Base=Multiply(A=NamedRerouteUsage()) {constB=2})
  n60 = Multiply(A=Multiply(A=Saturate(None=@n59), B=Add(A=@n16, B=@n40)) {constA=1}, B=@n1.A)
  StaticSwitchParameter(True=Multiply(A=@n60, B=@n42), False=@n60) {parameterName=UseDepthFade}
 ]
}
